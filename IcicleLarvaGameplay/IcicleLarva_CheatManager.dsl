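; Cheat manager for IcicleLarva gameplay that handles skydrop spawn requests.
; OnSkydropRequested takes a spawn location and rotation and builds a transform from them.
; It gets the IcicleLarvaFortGamestateComp component from the game state and calls its SkydropSpawnatTransform with that transform.
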(fn OnSkydropRequested
  (params (SpawnLocation :type Vector%20%28by%20ref%29 :default "0, 0, 0") (SpawnRotation :type Rotator%20%28by%20ref%29 :default "0, 0, 0"))
  (Class|IcicleLarvaFortGamestateComp|SkydropSpawnatTransform (Actor|GetComponentbyClass (Game|GetGameState) "/IcicleLarvaGameplay/Gameplay/IcicleLarvaFortGamestateComp.IcicleLarvaFortGamestateComp_C") (Math|Transform|MakeTransform SpawnLocation SpawnRotation))
  (return))
